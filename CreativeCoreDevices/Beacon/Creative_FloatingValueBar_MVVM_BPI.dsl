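; Blueprint interface for a Creative floating value bar widget, filed under the Beacon device and named for MVVM use.
; It declares setters for percentage, current and max value text, active widget index, fonts, icon and header colors, scale, spacer transform and visibility.
; It also declares an element-size query and an appear-animation call with play mode, both returning values.
; All functions are signatures only, leaving behavior to implementers.
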(fn GetElementSizes
  (return false 0 0))

(fn PlayAnimation_Appear
  (params (PlayMode :type EUMGSequencePlayMode%20Enum :default "Forward"))
  (return false 0.0))

(fn SetActiveWidgetIndex
  (params (IN_WidgetIndex :type Integer)))

(fn SetBackerIconVisibility
  (params (IN_Visibility :type ESlateVisibility%20Enum :default "Visible")))

(fn SetHeaderFontColor
  (params (IN_Color :type Slate%20Color%20Structure)))

(fn SetIconColors
  (params (IN_BackingColor :type Linear%20Color%20Structure) (IN_BorderColor :type Linear%20Color%20Structure) (IN_FillColor :type Linear%20Color%20Structure)))

(fn SetIconScaleOverride
  (params (Scale :type Float%20%28double-precision%29)))

(fn SetPercentage
  (params (IN_Percentage :type Float%20%28double-precision%29)))

(fn SetRenderTransformOfSpacer
  (params (IN_ScaleX :type Float%20%28double-precision%29)))

(fn SetScaleOverride
  (params (IN_Scale :type Vector%202D%20Structure)))

(fn SetValueText
  (params (IN_CurrentValue :type Text) (IN_MaxValue :type Text)))

(fn SetVisibilityOfValues
  (params (IN_Visibility :type ESlateVisibility%20Enum :default "Visible")))

(fn SetWidgetFont
  (params (IN_FontToUse :type Slate%20Font%20Info%20Structure)))
