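# Default UMG floating value bar for Creative Beacon devices, showing a progress icon with indicator image, switchable value, objective or distance text, and a direction arrow.
# Setter events drive icon material fill and colors, text color, visibility, scale and the active text page.
# GetElementSizes reports content and arrow box sizes for callers.
# PlayAnimation_Appear runs a sped-up pop-in that staggers icon and values, returning the animation's unscaled end time.
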WidgetBlueprint /CreativeCoreDevices/Beacon/Creative_FloatingValueBar_MVVM_Default
animations [Appear {
  binding_count=3
  length_seconds=0.7
  object=&self:Appear
  sequence {bindings [
    ProgressIcon {
     id=#1
     possessed_class=Image
     tracks [
      {
       type=MovieSceneFloatTrack
       display_name=Render Opacity
       sections [{
         type=MovieSceneFloatSection
         end_frame=10
         start_frame=0
         channels [None {type=MovieSceneScriptingFloatChannel, keys [[0,0],[9.789,1],[10.789,1]]}]
        }]
      }
      {
       type=MovieScene2DTransformTrack
       sections [{
         type=MovieScene2DTransformSection
         end_frame=10
         start_frame=0
         channels [
          Translation.X {type=MovieSceneScriptingFloatChannel}
          Translation.Y {type=MovieSceneScriptingFloatChannel}
          Angle {type=MovieSceneScriptingFloatChannel}
          Scale.X {type=MovieSceneScriptingFloatChannel, keys [[2.6977,-0.5],[7.789,1.05],[10.789,1]]}
          Scale.Y {type=MovieSceneScriptingFloatChannel, keys [[2.6977,0.5],[7.789,1.05],[10.789,1]]}
          Shear.X {type=MovieSceneScriptingFloatChannel}
          Shear.Y {type=MovieSceneScriptingFloatChannel}
         ]
        }]
      }
      {
       type=MovieSceneWidgetMaterialTrack
       sections [{
         type=MovieSceneParameterSection
         end_frame=10
         start_frame=0
         channels [OverrideFillColor {type=MovieSceneScriptingFloatChannel, keys [[0,1],[4.8253,0.9763],[5.327,-0.0237],[9.199,0],[10.789,0]]}]
        }]
      }
     ]
    }
    ValueBox {
     id=#2
     possessed_class=HorizontalBox
     tracks [
      {
       type=MovieSceneFloatTrack
       display_name=Render Opacity
       sections [{
         type=MovieSceneFloatSection
         end_frame=14
         start_frame=0
         channels [None {type=MovieSceneScriptingFloatChannel, keys [[0,0],[10.789,0],[14,1]]}]
        }]
      }
      {
       type=MovieScene2DTransformTrack
       sections [{
         type=MovieScene2DTransformSection
         end_frame=14
         start_frame=10
         channels [
          Translation.X {type=MovieSceneScriptingFloatChannel, const [0,2]}
          Translation.Y {type=MovieSceneScriptingFloatChannel, keys [[10,7],[14,0]]}
          Angle {type=MovieSceneScriptingFloatChannel}
          Scale.X {type=MovieSceneScriptingFloatChannel}
          Scale.Y {type=MovieSceneScriptingFloatChannel}
          Shear.X {type=MovieSceneScriptingFloatChannel}
          Shear.Y {type=MovieSceneScriptingFloatChannel}
         ]
        }]
      }
     ]
    }
    Overlay {
     id=#3
     possessed_class=Overlay
     tracks [
      {
       type=MovieSceneFloatTrack
       display_name=Render Opacity
       sections [{
         type=MovieSceneFloatSection
         end_frame=10
         start_frame=0
         channels [None {type=MovieSceneScriptingFloatChannel, keys [[0,0],[5,1],[10.789,1]]}]
        }]
      }
      {
       type=MovieScene2DTransformTrack
       sections [{
         type=MovieScene2DTransformSection
         end_frame=10
         start_frame=0
         channels [
          Translation.X {type=MovieSceneScriptingFloatChannel}
          Translation.Y {type=MovieSceneScriptingFloatChannel}
          Angle {type=MovieSceneScriptingFloatChannel}
          Scale.X {type=MovieSceneScriptingFloatChannel, keys [[0,0.3023],[3,0.8665],[4,1.4397],[6.139,1],[10.789,1]]}
          Scale.Y {type=MovieSceneScriptingFloatChannel, keys [[0,0.6],[3,1.5185],[4,0.8081],[6.139,1],[10.789,1]]}
          Shear.X {type=MovieSceneScriptingFloatChannel}
          Shear.Y {type=MovieSceneScriptingFloatChannel}
         ]
        }]
      }
     ]
    }
   ]}
 }]
graphs [
 EventGraph {nodes [
   n1 = Event SetIconScaleOverride!() -> (OutputDelegate, then, Scale:double) {is=title K2Node_Event}
   > SetRenderScale!(
    execute
    self:Widget Object Reference=Get Icon_IndicatorLabel() -> (Icon_IndicatorLabel:Image Object Reference)
    Scale:Vector 2D Structure=To Vector2D!(InDouble:double=@n1.Scale) -> (ReturnValue:Vector 2D Structure)
   ) -> (then)
   n2 = Event SetPercentage!() -> (OutputDelegate, then, IN_Percentage:double) {is=title K2Node_Event}
   > n3 = GetDynamicMaterial(self=Get ProgressIcon() -> (ProgressIcon:Image Object Reference))
   > SetScalarParameterValue(self=@n3, ParameterName=PercentFill, Value=@n2.IN_Percentage)
   n4 = Event SetScaleOverride!() -> (OutputDelegate, then, IN_Scale:Vector 2D Structure) {is=title K2Node_Event}
   > SetRenderScale(self=Get Overlay() -> (Overlay:Overlay Object Reference), Scale=@n4.IN_Scale)
   n5 = Event SetIconColors!() -> (OutputDelegate, then, IN_BackingColor:Linear Color Structure, IN_BorderColor:Linear Color Structure, IN_FillColor:Linear Color Structure) {is=title K2Node_Event}
   > n6 = GetDynamicMaterial(self=Get ProgressIcon())
   > SetVectorParameterValue(self=@n6, ParameterName=BackingColor, Value=@n5.IN_BackingColor)
   > SetVectorParameterValue(self=@n6, ParameterName=BorderColor, Value=@n5.IN_BorderColor)
   > SetVectorParameterValue(self=@n6, ParameterName=FillColor, Value=@n5.IN_FillColor)
   n7 = Event SetBackerIconVisibility!() -> (OutputDelegate, then, IN_Visibility:ESlateVisibility Enum=Visible) {is=title K2Node_Event}
   > SetVisibility(self=Get ProgressIcon(), InVisibility=@n7.IN_Visibility)
   n8 = Event SetHeaderFontColor!() -> (OutputDelegate, then, IN_Color:Slate Color Structure) {is=title K2Node_Event}
   > SetColorAndOpacity(
    self=[
     Get CurrentValue() -> (CurrentValue:Text Object Reference)
     Get MaxValue() -> (MaxValue:Text Object Reference)
     Get ObjectiveText() -> (ObjectiveText:Text Object Reference)
    ]
    InColorAndOpacity=@n8.IN_Color
   )
   n9 = Event SetVisibilityOfValues!() -> (OutputDelegate, then, IN_Visibility:ESlateVisibility Enum=Visible) {is=title K2Node_Event}
   > SetVisibility(self=[Get MaxValue(), Get ValueBoxSpacer() -> (ValueBoxSpacer:Image Object Reference), Get CurrentValue()], InVisibility=@n9.IN_Visibility)
   n10 = Event SetRenderTransformOfSpacer!() -> (OutputDelegate, then, IN_ScaleX:double) {is=title K2Node_Event}
   > SetRenderTransform!(
    execute
    self:Widget Object Reference=Get ValueBoxSpacer()
    InTransform_Translation:Vector 2D Structure=(X=0,Y=0)
    InTransform_Scale_X:double=@n10.IN_ScaleX
    InTransform_Scale_Y:double=1
    InTransform_Shear:Vector 2D Structure=(X=0,Y=0)
    InTransform_Angle:float=0
   ) -> (then)
   n11 = Event SetActiveWidgetIndex!() -> (OutputDelegate, then, IN_WidgetIndex:int=0) {is=title K2Node_Event}
   > SetActiveWidgetIndex(self=Get TextSwitch() -> (TextSwitch:Common Widget Switcher Legacy Object Reference), Index=@n11.IN_WidgetIndex)
  ]}
 GetElementSizes {nodes [
   GetElementSizes() {is=title K2Node_FunctionEntry}
   n1 = Get ObjectiveVerticalBox() -> (ObjectiveVerticalBox:Vertical Box Object Reference)
   n2 = Get ArrowSizeBox() -> (ArrowSizeBox:Size Box Object Reference)
   > Return Node!(
    execute
    Success:bool=true
    Size_VerticalBox:Vector 2D Structure=GetLocalSize!(Geometry:"Geometry Structure (by ref)"=GetCachedGeometry!(self:Widget Object Reference=@n1) -> (ReturnValue:"Geometry Structure (by ref)")) -> (ReturnValue:Vector 2D Structure)
    Size_ArrowBox:Vector 2D Structure=MakeVector2D(
     X=Get WidthOverride!(self:Size Box Object Reference=@n2) -> (WidthOverride:float)
     Y=Get HeightOverride!(self:Size Box Object Reference=@n2) -> (HeightOverride:float)
    )
   )
  ]}
 PlayAnimation_Appear {nodes [
   n1 = PlayAnimation_Appear!() -> (then, PlayMode:EUMGSequencePlayMode Enum=Forward) {is=title K2Node_FunctionEntry}
   > PlayAnimation(InAnimation=Get Appear() -> (Appear:Widget Animation Object Reference), PlayMode=@n1, PlaybackSpeed=1.5)
   > Return Node!(execute, Success=true, AnimationLength:double=GetEndTime(self=Get Appear()))
  ]}
]
variables []
widgets {
 info {inheritedWidgetCount=0, parentClass=&UserWidget, rootWidgetClass=&SizeBox, widgetCount=17}
 widgets [
  SizeBox_3 {widgetClassPath=&SizeBox, widget=SizeBox_3, properties {heightOverride=21.3612}}
  Overlay {widgetClassPath=&Overlay, bIsVariable=true, parent=SizeBox_3, widget=Overlay, slot {class=SizeBoxSlot}}
  ArrowSizeBox {
   widgetClassPath=&SizeBox
   bIsVariable=true
   parent=Overlay
   widget=ArrowSizeBox
   properties {bOverride_HeightOverride=true, bOverride_WidthOverride=true, heightOverride=160, widthOverride=160}
   slot {class=OverlaySlot}
  }
  ArrowSpinnerSizeBox {
   widgetClassPath=&SizeBox
   bIsVariable=true
   parent=ArrowSizeBox
   widget=ArrowSpinnerSizeBox
   properties {bOverride_WidthOverride=true, widthOverride=32}
   slot {class=SizeBoxSlot, properties {horizontalAlignment=HAlign_Center}}
  }
  Image_61 {
   widgetClassPath=&Image
   parent=ArrowSpinnerSizeBox
   widget=Image_61
   properties {brush {
     bIsDynamicallyLoaded=false
     drawAs=Image
     imageType=NoImage
     mirroring=NoMirror
     resourceName=None
     resourceObject=/Game/Widgets/SActorCanvas/DirectionArrow
     tiling=NoTile
     imageSize=<32 32>
     margin=<ltrb 0 0 0 0>
     outlineSettings {
      bUseBrushTransparency=false
      roundingType=HalfHeightRadius
      width=0
      color {colorUseRule=UseColor_Specified, specifiedColor=<rgba 0 0 0 0>}
      cornerRadii=<0 0 0 0>
     }
     tintColor {colorUseRule=UseColor_Specified, specifiedColor=<rgba 1 1 1 1>}
     uVRegion {bIsValid=false, max=<0 0>, min=<0 0>}
    }}
   slot {class=SizeBoxSlot, properties {horizontalAlignment=HAlign_Center, verticalAlignment=VAlign_Top}}
  }
  ObjectiveVerticalBox {
   widgetClassPath=&VerticalBox
   bIsVariable=true
   parent=Overlay
   widget=ObjectiveVerticalBox
   slot {class=OverlaySlot, properties {horizontalAlignment=HAlign_Center, verticalAlignment=VAlign_Center}}
  }
  TextSwitch {
   widgetClassPath=&CommonWidgetSwitcherLegacy
   bIsVariable=true
   parent=ObjectiveVerticalBox
   widget=TextSwitch
   slot {class=VerticalBoxSlot}
  }
  ValueBox {
   widgetClassPath=&HorizontalBox
   bIsVariable=true
   parent=TextSwitch
   widget=ValueBox
   properties {visibility=HitTestInvisible}
   slot {class=WidgetSwitcherSlot, properties {horizontalAlignment=HAlign_Center, padding=<ltrb 0 0 0 -2>}}
  }
  CurrentValue {
   widgetClassPath=&TextBlock
   bIsVariable=true
   parent=ValueBox
   widget=CurrentValue
   properties {
    justification=Right
    minDesiredWidth=20
    text=75
    colorAndOpacity {colorUseRule=UseColor_Specified, specifiedColor=<rgba 1 1 1 0.75>}
    font {
     bForceMonospaced=false
     bMaterialIsStencil=false
     fontMaterial=null
     fontObject=&NotoSans
     letterSpacing=0
     monospacedWidth=1
     size=10
     skewAmount=0
     typefaceFontName=NotoSans-Bold
     outlineSettings {
      bApplyOutlineToDropShadows=false
      bMiteredCorners=false
      bSeparateFillAlpha=false
      outlineMaterial=null
      outlineSize=0
      outlineColor=<rgba 0 0 0 1>
     }
    }
   }
   slot {class=HorizontalBoxSlot, properties {padding=<ltrb 0 0 2 0>}}
  }
  ValueBoxSpacer {
   widgetClassPath=&Image
   bIsVariable=true
   parent=ValueBox
   widget=ValueBoxSpacer
   properties {
    renderOpacity=0.5
    visibility=SelfHitTestInvisible
    brush {
     bIsDynamicallyLoaded=false
     drawAs=Image
     imageType=NoImage
     mirroring=NoMirror
     resourceName=None
     resourceObject=null
     tiling=NoTile
     imageSize=<2 6>
     margin=<ltrb 0 0 0 0>
     outlineSettings {
      bUseBrushTransparency=false
      roundingType=HalfHeightRadius
      width=0
      color {colorUseRule=UseColor_Specified, specifiedColor=<rgba 0 0 0 0>}
      cornerRadii=<0 0 0 0>
     }
     tintColor {colorUseRule=UseColor_Specified, specifiedColor=<rgba 1 1 1 1>}
     uVRegion {bIsValid=false, max=<0 0>, min=<0 0>}
    }
   }
   slot {class=HorizontalBoxSlot, properties {padding=<ltrb 0 1 0 2>}}
  }
  MaxValue {
   widgetClassPath=&TextBlock
   bIsVariable=true
   parent=ValueBox
   widget=MaxValue
   properties {
    minDesiredWidth=20
    text=75
    colorAndOpacity {colorUseRule=UseColor_Specified, specifiedColor=<rgba 1 1 1 0.75>}
    font {
     bForceMonospaced=false
     bMaterialIsStencil=false
     fontMaterial=null
     fontObject=&NotoSans
     letterSpacing=0
     monospacedWidth=1
     size=10
     skewAmount=0
     typefaceFontName=NotoSans-Bold
     outlineSettings {
      bApplyOutlineToDropShadows=false
      bMiteredCorners=false
      bSeparateFillAlpha=false
      outlineMaterial=null
      outlineSize=0
      outlineColor=<rgba 0 0 0 1>
     }
    }
   }
   slot {class=HorizontalBoxSlot, properties {padding=<ltrb 2 0 0 0>}}
  }
  ObjectiveText {
   widgetClassPath=&TextBlock
   bIsVariable=true
   parent=TextSwitch
   widget=ObjectiveText
   properties {
    justification=Center
    minDesiredWidth=20
    text=Objective
    colorAndOpacity {colorUseRule=UseColor_Specified, specifiedColor=<rgba 1 1 1 0.75>}
    font {
     bForceMonospaced=false
     bMaterialIsStencil=false
     fontMaterial=null
     fontObject=&NotoSans
     letterSpacing=0
     monospacedWidth=1
     size=10
     skewAmount=0
     typefaceFontName=NotoSans-Bold
     outlineSettings {
      bApplyOutlineToDropShadows=false
      bMiteredCorners=false
      bSeparateFillAlpha=false
      outlineMaterial=null
      outlineSize=0
      outlineColor=<rgba 0 0 0 1>
     }
    }
   }
   slot {class=WidgetSwitcherSlot, properties {horizontalAlignment=HAlign_Center, verticalAlignment=VAlign_Center}}
  }
  DistanceText {
   widgetClassPath=&TextBlock
   bIsVariable=true
   parent=TextSwitch
   widget=DistanceText
   properties {
    justification=Center
    minDesiredWidth=20
    text=0m
    font {
     bForceMonospaced=false
     bMaterialIsStencil=false
     fontMaterial=null
     fontObject=&Burbank
     letterSpacing=0
     monospacedWidth=1
     size=20
     skewAmount=0
     typefaceFontName=Burbank-PlayerPower
     outlineSettings {
      bApplyOutlineToDropShadows=false
      bMiteredCorners=false
      bSeparateFillAlpha=false
      outlineMaterial=null
      outlineSize=1
      outlineColor=<rgba 0 0 0 1>
     }
    }
   }
   slot {class=WidgetSwitcherSlot, properties {horizontalAlignment=HAlign_Center, verticalAlignment=VAlign_Center}}
  }
  Icon {widgetClassPath=&Overlay, parent=ObjectiveVerticalBox, widget=Icon, slot {class=VerticalBoxSlot}}
  ProgressIcon {
   widgetClassPath=&Image
   bIsVariable=true
   parent=Icon
   widget=ProgressIcon
   properties {
    brush {
     bIsDynamicallyLoaded=false
     drawAs=Image
     imageType=NoImage
     mirroring=NoMirror
     resourceName=None
     resourceObject=/Game/Athena/HUD/LTM/Playground/Materials/M_UI_HUD_Objective_Progress
     tiling=NoTile
     imageSize=<58 58>
     margin=<ltrb 0 0 0 0>
     outlineSettings {
      bUseBrushTransparency=false
      roundingType=HalfHeightRadius
      width=0
      color {colorUseRule=UseColor_Specified, specifiedColor=<rgba 0 0 0 0>}
      cornerRadii=<0 0 0 0>
     }
     tintColor {colorUseRule=UseColor_Specified, specifiedColor=<rgba 1 1 1 1>}
     uVRegion {bIsValid=false, max=<0 0>, min=<0 0>}
    }
    renderTransformPivot=<0.5 1>
   }
   slot {class=OverlaySlot, properties {horizontalAlignment=HAlign_Center, verticalAlignment=VAlign_Center}}
  }
  IndicatorColor {
   widgetClassPath=&Border
   bIsVariable=true
   parent=Icon
   widget=IndicatorColor
   properties {
    horizontalAlignment=HAlign_Center
    verticalAlignment=VAlign_Center
    background {
     bIsDynamicallyLoaded=false
     drawAs=NoDrawType
     imageType=NoImage
     mirroring=NoMirror
     resourceName=None
     resourceObject=null
     tiling=NoTile
     imageSize=<32 32>
     margin=<ltrb 0 0 0 0>
     outlineSettings {
      bUseBrushTransparency=false
      roundingType=HalfHeightRadius
      width=0
      color {colorUseRule=UseColor_Specified, specifiedColor=<rgba 0 0 0 0>}
      cornerRadii=<0 0 0 0>
     }
     tintColor {colorUseRule=UseColor_Specified, specifiedColor=<rgba 1 1 1 1>}
     uVRegion {bIsValid=false, max=<0 0>, min=<0 0>}
    }
    padding=<ltrb 0 0 0 5>
   }
   slot {class=OverlaySlot, properties {horizontalAlignment=HAlign_Center, verticalAlignment=VAlign_Center, padding=<ltrb 0 0 0 2>}}
  }
  Icon_IndicatorLabel {
   widgetClassPath=&Image
   bIsVariable=true
   parent=IndicatorColor
   widget=Icon_IndicatorLabel
   properties {renderTransformPivot=<0.5 1>}
   slot {class=BorderSlot, properties {horizontalAlignment=HAlign_Center, verticalAlignment=VAlign_Center, padding=<ltrb 0 0 0 5>}}
  }
 ]
}
