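# SingleIconMaterial is a translucent UI material in the Beacon device folder that draws a single tinted icon.
# It multiplies the RGB of an Icon texture parameter by the RGB of a Tint vector parameter, which defaults to white.
# The Icon parameter defaults to a texture from the Creative icon picker library, and both parameters can be overridden by material instances.
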Material /CreativeCoreDevices/Beacon/SingleIconMaterial
graph {
 properties {blendMode=BLEND_Translucent, materialDomain=MD_UI, pixelDepthOffsetMode=PDOM_Legacy, usedShadingModels=Default Lit}
 nodes [Multiply(
   A=VectorParameter() {parameterName=Tint, defaultValue=<rgba 1 1 1 1>}.RGB
   B=TextureSampleParameter2D() {parameterName=Icon, texture=/Game/Athena/UI/InGame/Creative/PropertyEditor/Icon_Picker/Textures/128x128/T_UI_IconLibrary_A_128}.RGB
  )]
}
