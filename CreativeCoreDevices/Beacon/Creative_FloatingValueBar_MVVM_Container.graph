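# Container widget for the Creative Beacon device's floating value bar or badge, showing icon, value, text, distance and offscreen arrow.
# It async-creates the component's child widget class (or a default) and binds a new Beacon view model to it.
# Component and team-change events refresh colors, text and visibility by team affiliation or custom color.
# Looping timers drive range and optional async line-of-sight checks with animated show/hide, plus distance and arrow updates.
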WidgetBlueprint /CreativeCoreDevices/Beacon/Creative_FloatingValueBar_MVVM_Container
graphs [
 Bind Owning Player {nodes [
   Bind Owning Player() {is=title K2Node_FunctionEntry}
   n1 = GetOwningPlayer()
   > Is Valid(InputObject=@n1)
   > n2 = Cast To FortPlayerControllerAthena(Object=@n1)
   > Bind Event to On Local Player Changed Teams(self:Fort Player Controller Athena Object Reference=@n2, Delegate=Create Event()) {is=title K2Node_AddDelegate}
  ]}
 Bind Widget Component {nodes [
   Bind Widget Component() {is=title K2Node_FunctionEntry}
   > n1 = Get!(execute:Exec) -> (then:Exec, else:Exec, Value Bar Component:Creative Value Bar Component Object Reference)
   > Bind Event to On Value Bar Component Updated(self:Creative Value Bar Component Object Reference=@n1, Delegate=Create Event()) {is=title K2Node_AddDelegate}
   > Bind Event to Custom Color Changed(self:Creative Value Bar Component Object Reference=@n1, Delegate=Create Event()) {is=title K2Node_AddDelegate}
  ]}
 CheckTeamUpdates {nodes [
   CheckTeamUpdates() {is=title K2Node_FunctionEntry}
   > Get()
   > n1 = GetLocalPlayerAffiliation!(execute, self) -> (then, LocalTeamAffiliation:EFortTeamAffiliation Enum=Friendly)
   n2 = Get Value Bar Component() -> (Value Bar Component:Creative Value Bar Component Object Reference)
   > n3 = Branch(Condition=NOT Boolean(A=Get Use Custom Color!(self:Creative Value Bar Component Object Reference=@n2) -> (Use Custom Color:bool)))
   > n4 = UpdateIconColorByAffiliation!(execute, self, TeamAffiliation:EFortTeamAffiliation Enum=@n1) -> (then)
   UpdateTextByAffiliation!(execute=[@n4.then, @n3.else], self, TeamAffiliation:EFortTeamAffiliation Enum=@n1) -> (then)
   > UpdateVisibilityByAffiliation!(execute, self, TeamAffiliation:EFortTeamAffiliation Enum=@n1) -> (then)
  ]}
 EventGraph {nodes [
   On Widget Component Updated!() -> (OutputDelegate, then, ValueBarComp:Creative Value Bar Component Object Reference) {is=title K2Node_CustomEvent}
   > n1 = Get()
   > UpdateVisibilityCheck()
   > Show Value!(execute, self, ValueIsOn:bool=Get Has Dynamic Value!(self:Creative Value Bar Component Object Reference=@n1) -> (Has Dynamic Value:bool)) -> (then)
   > CheckTeamUpdates() {is=title K2Node_CallFunction}
   > SetColor()
   > "Update Text && Indicator"()
   > n2 = Sequence!(execute) -> (then_0, then_1, then_2:Exec)
   > n3 = Branch(Condition=Get bDisplayDistanceText!(self:Creative Value Bar Component Object Reference=Get Value Bar Component()) -> (bDisplayDistanceText:bool))
   > Show Distance Text()
   Hide Distance Text(execute=@n3.else)
   n4 = Branch(
    execute=@n2.then_1
    Condition=Get bDisplayOffscreenArrow!(self:Creative Value Bar Component Object Reference=Get Value Bar Component()) -> (bDisplayOffscreenArrow:bool)
   )
   > Show Offscreen Arrow()
   Hide Offscreen Arrow(execute=@n4.else)
   n5 = Sequence!(execute=@n2.then_2) -> (then_0, then_1)
   SetRenderScale!(
    execute=@n5.then_1
    self:Widget Object Reference=Get Overlay() -> (Overlay:Overlay Object Reference)
    Scale:Vector 2D Structure=Get ScaleOverride!(self:Creative Value Bar Component Object Reference=Get Value Bar Component()) -> (ScaleOverride:Vector 2D Structure)
   ) -> (then)
   Set Progress Icon Visibility!(
    execute=@n5.then_0
    self
    bVisible:bool=Get bDisplayProgressIcon!(self:Creative Value Bar Component Object Reference=Get Value Bar Component()) -> (bDisplayProgressIcon:bool)
   ) -> (then) {is=title K2Node_CallFunction}
   UpdateVisibilityCheck() {is=title K2Node_CustomEvent}
   > n6 = Get()
   n7 = Get Visible Distance!(self:Creative Value Bar Component Object Reference=@n6) -> (Visible Distance:double)
   > n8 = Branch(Condition=AND Boolean(A="float >= float"(A=@n7, B=0) {is=title K2Node_CallFunction}, B=ShouldBeVisible!() -> (bShouldBeVisible:bool)))
   > n9 = Range Check!(execute, self) -> (then, ReturnValue:bool)
   > n10 = Set CurrentVisibility(CurrentVisibility:bool=@n9) -> (Output_Get:bool)
   Branch(
    execute=[@n26.then, @n10.then]
    Condition="Not Equal (Boolean)"!(A:bool=Get CurrentVisibility() -> (CurrentVisibility:bool), B:bool=Get LastVisibility() -> (LastVisibility:bool)) -> (ReturnValue:bool) {is=title K2Node_CallFunction}
   )
   > n11 = PlayAnimation_Appear!(
    execute:Exec
    self:Object Reference=Get MyChildWidget() -> (MyChildWidget:User Widget Object Reference)
    PlayMode:EUMGSequencePlayMode Enum=Select(Option 0:EUMGSequencePlayMode Enum=Reverse, Option 1:EUMGSequencePlayMode Enum, Index=Get CurrentVisibility()) -> (ReturnValue:EUMGSequencePlayMode Enum)
   ) -> (then:Exec, Success:bool, AnimationLength:double) {is=title K2Node_Message}
   > n12 = Set LastVisibility(LastVisibility:bool=Get CurrentVisibility()) -> (Output_Get:bool)
   > n13 = Branch(Condition=@n12)
   > SetVisibility(InVisibility=SelfHitTestInvisible)
   > Clear and Invalidate Timer by Handle(Handle=Get HideTimer() -> (HideTimer:Timer Handle Structure))
   n14 = Branch(execute=@n13.else, Condition=@n11)
   n15 = HideIndicator() {is=title K2Node_CustomEvent}
   n16 = Set Timer by Event(execute=@n14.then, Delegate=@n15, Time=@n11.AnimationLength, bLooping=false)
   > Set HideTimer(HideTimer:Timer Handle Structure=@n16) -> (Output_Get:Timer Handle Structure)
   SetVisibility(execute=[@n15.then, @n14.else], InVisibility=Collapsed)
   Disable HUD(execute=@n8.else)
   Clear and Invalidate Timer by Handle(execute=@n6.else, Handle=Get VisCheckTimerHandle() -> (VisCheckTimerHandle:Timer Handle Structure))
   Show Offscreen Arrow() {is=title K2Node_CustomEvent}
   > n17 = Set Timer by Event(Delegate=Create Event(), Time=Get OffscreenArrowTickTime() -> (OffscreenArrowTickTime:double), bLooping=true)
   > Set OffscreenArrowTimerHandle(OffscreenArrowTimerHandle:Timer Handle Structure=@n17) -> (Output_Get:Timer Handle Structure)
   Hide Offscreen Arrow() {is=title K2Node_CustomEvent}
   > Clear and Invalidate Timer by Handle(Handle=Get OffscreenArrowTimerHandle() -> (OffscreenArrowTimerHandle:Timer Handle Structure))
   > UpdateValue_ArrowVisilbility!(
    execute
    self:Device - Beacon View Model Object Reference=Get BadgeUIMVVM() -> (BadgeUIMVVM:Device - Beacon View Model Object Reference)
    Value:ESlateVisibility Enum=Hidden
   ) -> (then)
   > SetIdealDrawSize!(execute, self, UseArrowBox:bool=false) -> (then)
   Update Distance Text Timer() {is=title K2Node_CustomEvent}
   > n18 = Get!(execute) -> (then, else, My Actor:Actor Object Reference)
   n19 = GetOwningPlayerPawn!(self:User Widget Object Reference) -> (ReturnValue:Pawn Object Reference)
   > n20 = Is Valid(InputObject=@n19)
   n21 = float / float(A="Distance (Vector)"!(V1:Vector=Get Actor Location(self=@n19), V2:Vector=Get Actor Location(self=@n18)) -> (ReturnValue:double), B=100) {is=title K2Node_CallFunction}
   > UpdateValue_Text_DistanceText!(
    execute
    self:Device - Beacon View Model Object Reference=Get BadgeUIMVVM()
    Value:Text=Format Text!(Format="LOCTEXT(\"{Distance}m\")", Distance:int=Floor(A=@n21) {is=title K2Node_CallFunction}) -> (Result)
   ) -> (then)
   UpdateValue_Text_DistanceText(execute=[@n20.Is Not Valid, @n18.else], self=Get BadgeUIMVVM(), Value="LOCTEXT(\"?\")")
   Show Distance Text() {is=title K2Node_CustomEvent}
   > SetActiveWidgetIndex!(execute:Exec, self:Object Reference=Get MyChildWidget(), IN_WidgetIndex:int=2) -> (then:Exec) {is=title K2Node_Message}
   > n22 = Set Timer by Event(Delegate=Create Event(), Time=Get DistanceTickTime() -> (DistanceTickTime:double), bLooping=true)
   > Set DistanceUpdateTimerHandle(DistanceUpdateTimerHandle:Timer Handle Structure=@n22) -> (Output_Get:Timer Handle Structure)
   OnUIAndViewModelReady() {is=title K2Node_CustomEvent}
   > n23 = Sequence!(execute) -> (then_0, then_1, then_2:Exec, then_3:Exec, then_4:Exec)
   > Bind Owning Player() {is=title K2Node_CallFunction}
   > Bind Widget Component() {is=title K2Node_CallFunction}
   UpdateValue_ArrowVisilbility(execute=@n23.then_1, self=Get BadgeUIMVVM(), Value=Collapsed)
   > SetColor()
   > "Update Text && Indicator"()
   n24 = float / float(A=Get Visibility Initial Tick Max() -> (Visibility Initial Tick Max:double), B=2)
   n25 = Set Timer by Event(
    execute=@n23.then_2
    Delegate=Create Event()
    Time=Get Visibility Tick Time() -> (Visibility Tick Time:double)
    bLooping=true
    InitialStartDelay=@n24
    InitialStartDelayVariance=@n24
   )
   > Set VisCheckTimerHandle(VisCheckTimerHandle:Timer Handle Structure=@n25) -> (Output_Get:Timer Handle Structure)
   > CheckTeamUpdates()
   SetIdealDrawSize(execute=@n23.then_3, UseArrowBox=false)
   On Widget Component Updated!(execute=@n23.then_4, self, ValueBarComp:Creative Value Bar Component Object Reference) -> (then) {is=title K2Node_CallFunction}
   Disable HUD() {is=title K2Node_CustomEvent}
   > n26 = Set CurrentVisibility(CurrentVisibility=false)
   Show Damage for Duration() {is=title K2Node_CustomEvent}
   > Branch(Condition=Get HasText() -> (HasText:bool))
   > n27 = Show Health()
   n28 = Show Objective Text() {is=title K2Node_CustomEvent}
   Set Timer by Event(execute=@n27.then, Delegate=@n28, Time=1, bLooping=false)
   n29 = LoadIndicatorTexture!() -> (OutputDelegate, then, TextureToLoad:Texture 2D Soft Object Reference) {is=title K2Node_CustomEvent}
   > Branch(Condition=IsValidSoftObjectReference(SoftObjectReference=@n29.TextureToLoad))
   > n30 = Async Load Asset(Asset=@n29.TextureToLoad)
   n31 = Cast To Texture2D(execute=@n30.Completed, Object=@n30)
   > n32 = Is Valid(InputObject=Get MyIconMID() -> (MyIconMID:Material Instance Dynamic Object Reference))
   n33 = CreateDynamicMaterialInstance!(execute=@n32.Is Not Valid, Parent=&/SingleIconMaterial, OptionalName=None, CreationFlags=None) -> (then, ReturnValue)
   > n34 = Set MyIconMID(MyIconMID:Material Instance Dynamic Object Reference=@n33) -> (Output_Get:Material Instance Dynamic Object Reference)
   SetTextureParameterValue(execute=[@n32.Is Valid, @n34.then], self=Get MyIconMID(), ParameterName=Icon, Value=@n31)
   > UpdateValue_Indicator_Icon!(
    execute
    self:Device - Beacon View Model Object Reference=Get BadgeUIMVVM()
    Value:Material Instance Dynamic Object Reference=Get MyIconMID()
   ) -> (then)
   n35 = UpdateVisibilityTimer!() -> (OutputDelegate, then, Visible:bool=~) {is=title K2Node_CustomEvent}
   > n36 = Branch(Condition=@n35.Visible)
   n37 = Get VisCheckTimerHandle()
   > Unpause Timer by Handle!(execute, Handle:Timer Handle Structure=@n37) -> (then)
   Pause Timer by Handle!(execute=@n36.else, Handle:Timer Handle Structure=@n37) -> (then)
   Show Health() {is=title K2Node_CustomEvent}
   > SetActiveWidgetIndex(self=Get MyChildWidget(), IN_WidgetIndex=0)
   Hide Distance Text() {is=title K2Node_CustomEvent}
   > Update Text Data()
   > Clear and Invalidate Timer by Handle(Handle=Get DistanceUpdateTimerHandle() -> (DistanceUpdateTimerHandle:Timer Handle Structure))
   n38 = Event InitFromObject!() -> (OutputDelegate, then, InitObject:Object Reference) {is=title K2Node_Event}
   > n39 = Cast To Creative_ValueBar_Component!(execute, Object=@n38.InitObject) -> (then, CastFailed, AsCreative Value Bar Component:Creative Value Bar Component Object Reference) {is=title K2Node_DynamicCast}
   > n40 = Set Value Bar Component(Value Bar Component:Creative Value Bar Component Object Reference=@n39) -> (Output_Get:Creative Value Bar Component Object Reference)
   > Set My Actor(My Actor:Actor Object Reference=Get Actor Ref!(self:Creative Value Bar Component Object Reference=@n40) -> (Actor Ref:Actor Object Reference)) -> (Output_Get:Actor Object Reference)
   > n41 = Sequence!(execute) -> (then_0, then_1)
   n42 = Get My Actor() -> (My Actor:Actor Object Reference)
   n43 = GetComponentByClass!(self:Actor Object Reference=@n42, ComponentClass:Actor Component Class Reference=/Script/FortniteGame.FortMinigameObjectiveComponent) -> (ReturnValue:Fort Minigame Objective Component Object Reference)
   > n44 = Is Valid(InputObject=@n43)
   Set FortMinigameLogic(
    execute=@n44.Is Not Valid
    FortMinigameLogic:Fort Minigame Progress Component Object Reference=GetComponentByClass(self=@n42, ComponentClass=/Script/FortniteGame.FortMinigameProgressComponent) -> (ReturnValue:Fort Minigame Progress Component Object Reference)
   ) -> (Output_Get:Fort Minigame Progress Component Object Reference)
   Set Fort Minigame Objective(execute=@n44.Is Valid, Fort Minigame Objective:Fort Minigame Objective Component Object Reference=@n43) -> (Output_Get:Fort Minigame Objective Component Object Reference)
   n45 = Get CurrentChildClass!(self:Creative Value Bar Component Object Reference=Get Value Bar Component()) -> (CurrentChildClass:User Widget Soft Class Reference)
   n46 = CreateWidgetAsync!(
    execute:Exec=@n41.then_1
    UserWidgetSoftClass:User Widget Soft Class Reference=Select(
     Option 0:User Widget Soft Class Reference=&/Creative_FloatingValueBar_MVVM_Default.Creative_FloatingValueBar_MVVM_Default_C
     Option 1:User Widget Soft Class Reference=@n45
     Index=IsValidSoftClassReference!(SoftClassReference:"Object Soft Class Reference (by ref)"=@n45) -> (ReturnValue:bool)
    ) -> (ReturnValue:User Widget Soft Class Reference)
    OwningPlayer:Player Controller Object Reference
    bSuspendInputUntilComplete:bool=false
   ) -> (then:Exec, OnComplete:Exec, OnFailure:Exec, UserWidget:User Widget Object Reference) {is=title K2Node_AsyncAction}
   Set MyChildWidget(execute=@n46.OnComplete, MyChildWidget:User Widget Object Reference=@n46) -> (Output_Get:User Widget Object Reference)
   > AddChildToOverlay!(execute, self:Overlay Object Reference=Get Overlay(), Content:Widget Object Reference=@n46) -> (then, ReturnValue:Overlay Slot Object Reference)
   > n47 = Construct Object from Class(Class=&/UEFN_BeaconBadge_ViewModel.UEFN_BeaconBadge_ViewModel_C) -> (ReturnValue:Device - Beacon View Model Object Reference)
   > n48 = Set BadgeUIMVVM(BadgeUIMVVM:Device - Beacon View Model Object Reference=@n47) -> (Output_Get:Device - Beacon View Model Object Reference)
   > SetViewModel!(
    execute
    self:MVVM View Object Reference=Get View From User Widget!(
     self:Viewmodel Engine Subsytem Object Reference=Get MVVMSubsystem() -> (ReturnValue:Viewmodel Engine Subsytem Object Reference) {is=title K2Node_GetEngineSubsystem}
     UserWidget:User Widget Object Reference=@n46
    ) -> (ReturnValue:MVVM View Object Reference) {is=title K2Node_CallFunction}
    ViewModelName:Name=UEFN_BeaconBadge_ViewModel
    ViewModel:Notify Field Value Changed Interface=@n48
   ) -> (then, ReturnValue:bool)
   > OnUIAndViewModelReady() {is=title K2Node_CallFunction}
   SetActiveWidgetIndex(execute=@n28.then, self=Get MyChildWidget(), IN_WidgetIndex=1)
  ]}
 GetLocalPlayerAffiliation {nodes [
   GetLocalPlayerAffiliation() {is=title K2Node_FunctionEntry}
   > n1 = Branch(Condition=Get bUseOwnerAffiliation() -> (bUseOwnerAffiliation:bool))
   > n2 = Set localTeamIndex(localTeamIndex:Byte=GetActorPvPTeamIndex!(Actor:Actor Object Reference=GetValidPlayerPawn!(self) -> (FoundPawn:Pawn Object Reference)) -> (TeamIndex:Byte=0, ReturnValue:bool)) -> (Output_Get:Byte=0)
   n3 = Get localTeamIndex() -> (localTeamIndex:Byte=0)
   n4 = Get Value Bar Component()
   n5 = Branch(
    execute=[@n2.then, @n12.then]
    Condition=Contains Item(
     TargetArray:Array of Integers=Get Friendly Teams!(self:Creative Value Bar Component Object Reference=@n4) -> (Friendly Teams:Array of Integers)
     ItemToFind:"Integer (by ref)"="To Integer (Byte)"!(InByte:Byte=@n3) -> (ReturnValue:int=0)
    )
   )
   > Return Node!(execute, LocalTeamAffiliation:EFortTeamAffiliation Enum=Friendly)
   n6 = Get localTeamIndex() -> (localTeamIndex=0)
   n7 = Get Value Bar Component()
   n8 = Branch(
    execute=@n5.else
    Condition=Contains Item(
     TargetArray=Get Hostile Teams!(self:Creative Value Bar Component Object Reference=@n7) -> (Hostile Teams:Array of Integers)
     ItemToFind="To Integer (Byte)"(InByte=@n6) -> (ReturnValue=0)
    )
   )
   > Return Node(LocalTeamAffiliation=Hostile)
   n9 = Get My Actor()
   n10 = GetValidPlayerPawn()
   n11 = Branch(execute=@n8.else, Condition=AND Boolean(A=IsValid(Object=@n10), B=IsValid(Object=@n9)))
   > Return Node(LocalTeamAffiliation=GetTeamAffiliation!(ActorA:Actor Object Reference=GetValidPlayerPawn(), ActorB:Actor Object Reference=Get My Actor()) -> (ReturnValue:EFortTeamAffiliation Enum=Friendly))
   Return Node(execute=@n11.else, LocalTeamAffiliation=Neutral)
   n12 = Set localTeamIndex(
    execute=@n1.else
    localTeamIndex=GetActorPvPTeamIndex(Actor=GetFirstLocalFortPlayerController!(bIncludeNonSpectators:bool=true, bIncludeSpectators:bool=false) -> (ReturnValue:Fort Player Controller Object Reference)) -> (TeamIndex=0)
   ) -> (Output_Get=0)
  ]}
 GetValidPlayerPawn {nodes [
   GetValidPlayerPawn() {is=title K2Node_FunctionEntry}
   > n1 = Cast To Pawn!(execute, Object=GetViewTarget!(self:Controller Object Reference=GetOwningPlayer()) -> (ReturnValue:Actor Object Reference)) -> (then, CastFailed, AsPawn:Pawn Object Reference) {is=title K2Node_DynamicCast}
   > Return Node!(execute, FoundPawn:Pawn Object Reference=@n1)
   Return Node(execute=@n1.CastFailed)
  ]}
 Range Check {nodes [
   Range Check() {is=title K2Node_FunctionEntry}
   > n1 = Sequence()
   > n2 = Get!(execute) -> (then, else, Value Bar Component:Creative Value Bar Component Object Reference)
   > n3 = Get!(execute) -> (then, else, My Actor:Actor Object Reference)
   n4 = GetValidPlayerPawn() {is=title K2Node_CallFunction}
   n5 = Get Visible Distance(self=@n2)
   > n6 = Branch(Condition=AND Boolean!(A=IsValid(Object=@n3), B="float > float"(A=@n5, B=0) {is=title K2Node_CallFunction}, C:bool=IsValid(Object=@n4)) -> (ReturnValue))
   n7 = Get Visible Distance(self=Get Value Bar Component())
   n8 = vector - vector!(A:Vector=Get Actor Location(self=Get My Actor()), B:Vector=Get Actor Location(self=GetValidPlayerPawn())) -> (ReturnValue:Vector) {is=title K2Node_CallFunction}
   > n9 = Set local In Range(local In Range:bool="float >= float"(A=Square!(A:double=@n7) -> (ReturnValue:double), B=Vector Length Squared!(A:Vector=@n8) -> (ReturnValue:double))) -> (Output_Get:bool)
   n10 = Get Value Bar Component()
   > Branch(Condition=AND Boolean!(A=@n9, B=Get Require LOS!(self:Creative Value Bar Component Object Reference=@n10) -> (Require LOS:bool)) -> (ReturnValue))
   > Request LOS Update()
   > Set local In Range(local In Range=Get isInLineOfSight() -> (isInLineOfSight:bool))
   SetVisibility(execute=@n6.else, InVisibility=Collapsed)
   > Set local In Range(local In Range=false)
   Return Node!(execute=@n1.then_1, ReturnValue:bool=Get local In Range() -> (local In Range:bool))
  ]}
 Request LOS Update {nodes [
   Request LOS Update() {is=title K2Node_FunctionEntry}
   > n1 = GetAsyncTraceQueue!(execute, ContextActor:Actor Object Reference=GetOwningPlayer()) -> (then, ReturnValue:Async Trace Queue Object Reference)
   > n2 = Get!(execute) -> (then, else, Value Bar Component:Creative Value Bar Component Object Reference)
   > n3 = Get!(execute) -> (then, else, My Actor:Actor Object Reference)
   n4 = Get Attached Mesh!(self:Creative Value Bar Component Object Reference=@n2) -> (Attached Mesh:Static Mesh Component Object Reference)
   n5 = GetBoundingBox!(self:Static Mesh Object Reference=Get StaticMesh(self=@n4)) -> (ReturnValue_Min:Vector=~, ReturnValue_Max:Vector=~, ReturnValue_IsValid:bool=~)
   n6 = vector + vector(A:Vector=@n5, B:Vector=@n5.ReturnValue_Max) -> (ReturnValue:Vector) {is=title K2Node_CommutativeAssociativeBinaryOperator}
   n7 = Get World Scale!(self:Scene Component Object Reference=@n4) -> (ReturnValue:Vector) {is=title K2Node_CallFunction}
   > ThrottledAsyncTrace_SourceTarget!(
    execute
    self:Async Trace Queue Object Reference=@n1
    bUseLocalCameraAsSource:bool=true
    SourceActor:Actor Object Reference
    SourceOffset_X:double=0
    SourceOffset_Y:double=0
    SourceOffset_Z:double=0
    TargetActor:Actor Object Reference=@n3
    TargetOffset:"Vector (by ref)"=vector / float!(
     A:Vector=vector * vector(A:Vector=@n6, B:Vector=@n7) -> (ReturnValue:Vector) {is=title K2Node_CommutativeAssociativeBinaryOperator}
     B:double=2
    ) -> (ReturnValue:Vector)
    ActorsToIgnore:Array of Actor Object References=Make Array!("[0]":Actor Object Reference=GetValidPlayerPawn(), "[1]":Actor Object Reference=Get My Actor()) -> (Array:Array of Actor Object References)
    CollisionChannel:ECollisionChannel Enum=ECC_Camera
    CompletionFunction:Delegate=Create Event()
   ) -> (then)
  ]}
 Set Progress Icon Visibility {nodes [
   n1 = Set Progress Icon Visibility!() -> (then, bVisible:bool=~) {is=title K2Node_FunctionEntry}
   > SetBackerIconVisibility!(
    execute:Exec
    self:Object Reference=Get MyChildWidget()
    IN_Visibility:ESlateVisibility Enum=Select(Option 0:ESlateVisibility Enum=Hidden, Option 1:ESlateVisibility Enum=Visible, Index=@n1) -> (ReturnValue:ESlateVisibility Enum)
   ) -> (then:Exec) {is=title K2Node_Message}
   > SetIconScaleOverride!(
    execute:Exec
    self:Object Reference=Get MyChildWidget()
    Scale:double=SelectFloat(A=1, B=Get IconScaleWhenBackerHidden() -> (IconScaleWhenBackerHidden:double), bPickA=@n1)
   ) -> (then:Exec) {is=title K2Node_Message}
  ]}
 SetColor {nodes [
   SetColor() {is=title K2Node_FunctionEntry}
   > n1 = Branch(Condition=Get Use Custom Color(self=Get Value Bar Component()))
   > Set MyCustomColor(MyCustomColor:Linear Color Structure=Get Custom Color!(self:Creative Value Bar Component Object Reference=Get Value Bar Component()) -> (Custom Color:Linear Color Structure)) -> (Output_Get:Linear Color Structure)
   > Set DistanceToWhite(DistanceToWhite:float="Distance (LinearColor)"!(C1:Linear Color Structure=Get MyCustomColor() -> (MyCustomColor:Linear Color Structure), C2:Linear Color Structure=(R=1,G=1,B=1,A=1)) -> (ReturnValue:float)) -> (Output_Get:float)
   > Set Color_Backing(Color_Backing:Linear Color Structure=LinearColor - LinearColor(A:Linear Color Structure=Get MyCustomColor(), B=Get CustomColor_BackingDifference() -> (CustomColor_BackingDifference:double)) -> (ReturnValue:Linear Color Structure) {is=title K2Node_PromotableOperator}) -> (Output_Get:Linear Color Structure)
   n2 = Get CustomColor_BorderSwitchAmount() -> (CustomColor_BorderSwitchAmount:double)
   n3 = Get DistanceToWhite() -> (DistanceToWhite:float)
   n4 = Get CustomColor_BorderDifference() -> (CustomColor_BorderDifference:double)
   n5 = Get MyCustomColor()
   > Set Color_Border(Color_Border:Linear Color Structure=SelectColor(
     A=LinearColor - LinearColor(A=@n5, B=@n4)
     B=LinearColor + LinearColor(A:Linear Color Structure=@n5, B=@n4) -> (ReturnValue:Linear Color Structure) {is=title K2Node_PromotableOperator}
     bPickA="float < float"(A=@n3, B=@n2)
    )) -> (Output_Get:Linear Color Structure)
   > Set Color_Fill(Color_Fill:Linear Color Structure=Get MyCustomColor()) -> (Output_Get:Linear Color Structure)
   n6 = SelectColor(
    A=(R=0.1441,G=0.1441,B=0.1441,A=1)
    B=(R=1,G=1,B=1,A=1)
    bPickA="float < float"(A=Get DistanceToWhite(), B=Get CustomColor_IndicatorSwitchAmount() -> (CustomColor_IndicatorSwitchAmount:double))
   )
   n7 = Get MyCustomColor()
   n8 = Get Value Bar Component()
   > Set Color_Indicator(Color_Indicator:Linear Color Structure=Select(
     Option 0:Linear Color Structure=Get Color_Fill() -> (Color_Fill:Linear Color Structure)
     Option 1:Linear Color Structure="Lerp (LinearColor)"(A=@n7, B=@n6, Alpha=0.9)
     Index=Get bDisplayProgressIcon(self=@n8)
    ) -> (ReturnValue:Linear Color Structure)) -> (Output_Get:Linear Color Structure)
   > UpdateValue_Colors!(
    execute
    self:Device - Beacon View Model Object Reference=Get BadgeUIMVVM()
    Color_Backing:Linear Color Structure=Get Color_Backing() -> (Color_Backing:Linear Color Structure)
    Color_Border:Linear Color Structure=Get Color_Border() -> (Color_Border:Linear Color Structure)
    Color_Fill:Linear Color Structure=Get Color_Fill()
    Color_Indicator:Linear Color Structure=Get Color_Indicator() -> (Color_Indicator:Linear Color Structure)
   ) -> (then)
   > SetIconColors!(
    execute:Exec
    self:Object Reference=Get MyChildWidget()
    IN_BackingColor:Linear Color Structure=Get Color_Backing()
    IN_BorderColor:Linear Color Structure=Get Color_Border()
    IN_FillColor:Linear Color Structure=Get Color_Fill()
   ) -> (then:Exec) {is=title K2Node_Message}
   > Set IdealTextColor(IdealTextColor:Linear Color Structure=Get Color_Indicator()) -> (Output_Get:Linear Color Structure)
   n9 = GetLocalPlayerAffiliation(execute=@n1.else) -> (LocalTeamAffiliation=Friendly) {is=title K2Node_CallFunction}
   > UpdateIconColorByAffiliation(TeamAffiliation=@n9)
  ]}
 SetIdealDrawSize {nodes [
   SetIdealDrawSize!() -> (then, UseArrowBox:bool=~) {is=title K2Node_FunctionEntry}
   > n1 = GetElementSizes!(execute:Exec, self:Object Reference=Get MyChildWidget()) -> (then:Exec, Success:bool, Size_VerticalBox:Vector 2D Structure, Size_ArrowBox:Vector 2D Structure) {is=title K2Node_Message}
   n2 = GetCachedGeometry!(self:Widget Object Reference=Get Overlay()) -> (ReturnValue:"Geometry Structure (by ref)")
   n3 = Get UseArrowBox() -> (UseArrowBox:bool)
   > SetDrawSize!(
    execute
    self:Widget Component Object Reference=Get Value Bar Component()
    Size:Vector 2D Structure=Select(
     Option 0:Vector 2D Structure=GetLocalSize!(Geometry:"Geometry Structure (by ref)"=@n2) -> (ReturnValue:Vector 2D Structure)
     Option 1:Vector 2D Structure=Select(Option 0=@n1.Size_VerticalBox, Option 1=@n1.Size_ArrowBox, Index=@n3)
     Index=@n1
    ) -> (ReturnValue:Vector 2D Structure)
   ) -> (then)
  ]}
 ShouldBeVisible {nodes [
   ShouldBeVisible() {is=title K2Node_FunctionEntry}
   > n1 = Get!(execute) -> (then, else, Value Bar Component:Creative Value Bar Component Object Reference)
   > Return Node!(execute, bShouldBeVisible:bool=IsShowingValueBarAllowed!(self:Creative Value Bar Component Object Reference=@n1) -> (IsAllowed:bool))
   Return Node(execute=@n1.else, bShouldBeVisible=false)
  ]}
 Show Value {nodes [
   n1 = Show Value!() -> (then, ValueIsOn:bool=~) {is=title K2Node_FunctionEntry}
   > SetVisibilityOfValues!(
    execute:Exec
    self:Object Reference=Get MyChildWidget()
    IN_Visibility:ESlateVisibility Enum=Select(Option 0:ESlateVisibility Enum=Hidden, Option 1:ESlateVisibility Enum=SelfHitTestInvisible, Index=@n1) -> (ReturnValue:ESlateVisibility Enum)
   ) -> (then:Exec) {is=title K2Node_Message}
   > n2 = Branch(Condition=Get ValueIsOn() -> (ValueIsOn:bool))
   SetPercentage!(execute:Exec=@n2.else, self:Object Reference=Get MyChildWidget(), IN_Percentage:double=1) -> (then:Exec) {is=title K2Node_Message}
  ]}
 TraceComplete {nodes [
   n1 = TraceComplete!() -> (then, Success:bool=~, ActorHit:Actor Object Reference, HitResult:Hit Result Structure) {is=title K2Node_FunctionEntry}
   > Branch(Condition=@n1)
   > n2 = Branch(Condition=IsValid(Object=@n1.ActorHit))
   > n3 = Get!(execute) -> (then, else, My Actor:Actor Object Reference)
   > n4 = Branch(Condition="Equal (Object)"(A=@n3, B=@n1.ActorHit) {is=title K2Node_CallFunction})
   Set isInLineOfSight(execute=[@n2.else, @n4.then], isInLineOfSight:bool=true) -> (Output_Get:bool)
   Set isInLineOfSight(execute=@n4.else, isInLineOfSight=false)
  ]}
 Update {nodes [
   n1 = Update!() -> (then, CurrentValue:int, MaxValue:int, ForcedValuePercent:double=-1) {is=title K2Node_FunctionEntry}
   n2 = "Not Equal (Float)"!(A:double=@n1.ForcedValuePercent, B:double=-1) -> (ReturnValue:bool) {is=title K2Node_CallFunction}
   n3 = Truncate(A=int * float!(A:int=@n1.MaxValue, B:double=@n1.ForcedValuePercent) -> (ReturnValue:double))
   > SetValueText!(
    execute:Exec
    self:Object Reference=Get MyChildWidget()
    IN_CurrentValue:Text="To Text (Integer)"(Value=Select(Option 0:int=@n1, Option 1:int=@n3, Index=@n2) -> (ReturnValue:int))
    IN_MaxValue:Text="To Text (Integer)"(Value=Get MaxValue() -> (MaxValue:int=0))
   ) -> (then:Exec) {is=title K2Node_Message}
   n4 = "To Float (Integer)"(InInt=Get MaxValue() -> (MaxValue=0))
   n5 = "To Float (Integer)"(InInt=Get CurrentValue() -> (CurrentValue:int=0))
   > n6 = Set LocalValuePercent(LocalValuePercent:double=Select(Option 0:double=float / float(A=@n5, B=@n4), Option 1:double=Get ForcedValuePercent() -> (ForcedValuePercent:double), Index=@n2) -> (ReturnValue:double)) -> (Output_Get:double)
   > n7 = Set CurrentValuePercent(CurrentValuePercent:double=@n6) -> (Output_Get:double)
   > SetPercentage(self=Get MyChildWidget(), IN_Percentage=@n7)
  ]}
 Update Offscreen Arrow {nodes [
   Update Offscreen Arrow() {is=title K2Node_FunctionEntry}
   > n1 = Get!(execute) -> (then, else, Value Bar Component:Creative Value Bar Component Object Reference)
   > Branch(Condition=Get bClampProjectedLocalPositionToViewport!(self:Fort Widget Component Object Reference=@n1) -> (bClampProjectedLocalPositionToViewport:bool))
   > Set ViewportSize(ViewportSize:Vector 2D Structure=GetViewportSize!() -> (ReturnValue:Vector 2D Structure)) -> (Output_Get:Vector 2D Structure)
   n2 = Get!(self:Fort Widget Component Object Reference=Get Value Bar Component()) -> (PreClampedLocalPosition_X:double, PreClampedLocalPosition_Y:double)
   n3 = GetViewportScale!() -> (ReturnValue:float)
   n4 = Get!() -> (ViewportSize_X:double, ViewportSize_Y:double)
   n5 = float / float(A=@n4, B=@n3) {is=title K2Node_PromotableOperator}
   n6 = Get!(self:Fort Widget Component Object Reference=Get Value Bar Component()) -> (PreClampedLocalPosition_X:double, PreClampedLocalPosition_Y:double)
   n7 = float / float(A=@n4.ViewportSize_Y, B=@n3)
   > Set bOutOfViewportBounds(bOutOfViewportBounds:bool=OR Boolean!(
     A="float < float"!(A:double=@n2, B:double=0) -> (ReturnValue:bool) {is=title K2Node_CallFunction}
     B="float < float"(A=@n2.PreClampedLocalPosition_Y, B=0)
     C:bool="float > float"(A=@n6, B=@n5)
     D:bool="float > float"(A=@n6.PreClampedLocalPosition_Y, B=@n7)
    ) -> (ReturnValue)) -> (Output_Get:bool)
   n8 = AND Boolean(A=Get bOutOfViewportBounds() -> (bOutOfViewportBounds:bool), B=Get bDisplayOffscreenArrow(self=Get Value Bar Component()))
   > UpdateValue_ArrowVisilbility(
    self=Get BadgeUIMVVM()
    Value=Select(Option 0:ESlateVisibility Enum=Hidden, Option 1:ESlateVisibility Enum=SelfHitTestInvisible, Index=@n8) -> (ReturnValue:ESlateVisibility Enum)
   )
   > Branch(Condition=@n8)
   > SetIdealDrawSize(UseArrowBox=Get bOutOfViewportBounds()) {is=title K2Node_CallFunction}
   > Branch(Condition=Get bOutOfViewportBounds())
   n9 = Get(self=Get Value Bar Component())
   > n10 = Branch(Condition="float < float"(A=@n9, B=0))
   > UpdateValue_ArrowRotationAngle!(execute, self:Device - Beacon View Model Object Reference=Get BadgeUIMVVM(), Value:double=-90) -> (then)
   n11 = Branch(execute=@n10.else, Condition="float < float"(A=@n9.PreClampedLocalPosition_Y, B=0))
   > UpdateValue_ArrowRotationAngle(self=Get BadgeUIMVVM(), Value=0)
   n12 = Get(self=Get Value Bar Component())
   n13 = Get!() -> (ViewportSize_X:double, ViewportSize_Y:double)
   n14 = Branch(execute=@n11.else, Condition="float > float"(A=@n12, B=@n13))
   > UpdateValue_ArrowRotationAngle(self=Get BadgeUIMVVM(), Value=90)
   Branch(execute=@n14.else, Condition="float > float"(A=@n12.PreClampedLocalPosition_Y, B=@n13.ViewportSize_Y))
   > UpdateValue_ArrowRotationAngle(self=Get BadgeUIMVVM(), Value=180)
  ]}
 "Update Text && Indicator" {nodes [
   "Update Text && Indicator"() {is=title K2Node_FunctionEntry}
   > n1 = Sequence()
   n2 = Get Value Bar Component()
   n3 = Get TextScale!(self:Creative Value Bar Component Object Reference=@n2) -> (TextScale:double)
   SetRenderTransformOfSpacer!(execute:Exec=@n1.then_1, self:Object Reference=Get MyChildWidget(), IN_ScaleX:double=@n3) -> (then:Exec) {is=title K2Node_Message}
   n4 = Get FontStyle!(self:Creative Value Bar Component Object Reference=@n2) -> (FontStyle:EBeaconFontStyle Enum=NewEnumerator0)
   n5 = Find!(
    TargetMap:Map of EBeaconFontStyle Enums to Slate Font Info Structures=Get FontStyleMap() -> (FontStyleMap:Map of EBeaconFontStyle Enums to Slate Font Info Structures=NewEnumerator0)
    Key:"EBeaconFontStyle Enum (by ref)"=@n4
   ) -> (
    Value_FontObject:Object Reference
    Value_FontMaterial:Object Reference
    Value_OutlineSettings:Font Outline Settings Structure
    Value_TypefaceFontName:Name
    Value_Size:float=~
    Value_LetterSpacing:int
    Value_SkewAmount:float=~
    Value_bForceMonospaced:bool=~
    Value_bMaterialIsStencil:bool=~
    Value_MonospacedWidth:float=~
    ReturnValue
   )
   n6 = Truncate(A=float * float(A=@n3, B=10))
   > SetWidgetFont!(
    execute:Exec
    self:Object Reference=Get MyChildWidget()
    IN_FontToUse:Slate Font Info Structure=MakeSlateFontInfo!(
     FontObject:Object Reference=@n5
     FontMaterial:Object Reference=@n5.Value_FontMaterial
     OutlineSettings:Font Outline Settings Structure=@n5.Value_OutlineSettings
     TypefaceFontName:Name=@n5.Value_TypefaceFontName
     Size:float="To Float (Integer)"(InInt=@n6)
     LetterSpacing:int=@n5.Value_LetterSpacing
     SkewAmount:float=@n5.Value_SkewAmount
     bForceMonospaced:bool=false
     bMaterialIsStencil:bool=false
     MonospacedWidth:float=1
    ) -> (ReturnValue:Slate Font Info Structure)
   ) -> (then:Exec) {is=title K2Node_Message}
   n7 = Find!(
    TargetMap:Map of EBeaconFontStyle Enums to Slate Color Structures=Get FontColorMap() -> (FontColorMap:Map of EBeaconFontStyle Enums to Slate Color Structures=NewEnumerator0)
    Key=@n4
   ) -> (Value:Slate Color Structure, ReturnValue)
   n8 = Break Slate Color!(SlateColor:"Slate Color Structure (by ref)"=@n7) -> (
    SpecifiedColor_R:float=~
    SpecifiedColor_G:float=~
    SpecifiedColor_B:float=~
    SpecifiedColor_A:float=~
    ColorUseRule:ESlateColorStylingMode Enum
   ) {is=title K2Node_BreakStruct}
   n9 = Get Value Bar Component()
   n10 = Get!() -> (IdealTextColor_R:float=~, IdealTextColor_G:float=~, IdealTextColor_B:float=~, IdealTextColor_A:float=~)
   > SetHeaderFontColor!(
    execute:Exec
    self:Object Reference=Get MyChildWidget()
    IN_Color:Slate Color Structure=Select!(
     Option 0:Slate Color Structure=@n7
     Option 1_SpecifiedColor_R:float=@n10
     Option 1_SpecifiedColor_G:float=@n10.IdealTextColor_G
     Option 1_SpecifiedColor_B:float=@n10.IdealTextColor_B
     Option 1_SpecifiedColor_A:float=@n8.SpecifiedColor_A
     Option 1_ColorUseRule:ESlateColorStylingMode Enum=@n8.ColorUseRule
     Index=Get Use Custom Color(self=@n9)
    ) -> (ReturnValue:Slate Color Structure)
   ) -> (then:Exec) {is=title K2Node_Message}
   n11 = Get Indicator Texture!(self:Creative Value Bar Component Object Reference=Get Value Bar Component()) -> (Indicator Texture:Texture 2D Soft Object Reference)
   Branch(
    execute=@n1.then_0
    Condition="Not Equal (SoftObjectReference)"(
     A:"Soft Object Reference (by ref)"=Get MyCurrentIndicatorTexture() -> (MyCurrentIndicatorTexture:Texture 2D Soft Object Reference)
     B:"Soft Object Reference (by ref)"=@n11
    ) -> (ReturnValue:bool=~) {is=title K2Node_PromotableOperator}
   )
   > n12 = Set MyCurrentIndicatorTexture(MyCurrentIndicatorTexture:Texture 2D Soft Object Reference=@n11) -> (Output_Get:Texture 2D Soft Object Reference)
   > LoadIndicatorTexture!(execute, self, TextureToLoad:Texture 2D Soft Object Reference=@n12) -> (then) {is=title K2Node_CallFunction}
  ]}
 Update Text Data {nodes [
   Update Text Data() {is=title K2Node_FunctionEntry}
   > n1 = Branch(Condition=Get bDisplayDistanceText(self=Get Value Bar Component()))
   n2 = Branch(execute=@n1.else, Condition=Get HasText())
   > Show Objective Text() {is=title K2Node_CallFunction}
   Show Health(execute=@n2.else) {is=title K2Node_CallFunction}
  ]}
 UpdateIconColorByAffiliation {nodes [
   n1 = UpdateIconColorByAffiliation!() -> (then, TeamAffiliation:EFortTeamAffiliation Enum=Friendly) {is=title K2Node_FunctionEntry}
   n2 = Break Creative Floating Health Color!(Creative_FloatingHealthColor:"Creative Floating Health Color Structure (by ref)"=Find(
     TargetMap:Map of EFortTeamAffiliation Enums to Creative Floating Health Color Structures=Get TeamColors() -> (TeamColors:Map of EFortTeamAffiliation Enums to Creative Floating Health Color Structures=Friendly)
     Key:"EFortTeamAffiliation Enum (by ref)"=@n1
    ) -> (Value:Creative Floating Health Color Structure)) -> (
    BackingColor_2_#1:Linear Color Structure
    BorderColor_4_#2:Linear Color Structure
    FillColor_6_#3:Linear Color Structure
    ContentColorAndOpacity_8_#4:Linear Color Structure
   ) {is=title K2Node_BreakStruct}
   n3 = Select!(
    Option 0:Linear Color Structure=@n2.FillColor_6_#3
    Option 1:Linear Color Structure=@n2.ContentColorAndOpacity_8_#4
    Index=Get bDisplayProgressIcon(self=Get Value Bar Component())
   ) -> (ReturnValue:Linear Color Structure)
   > UpdateValue_Colors(
    self=Get BadgeUIMVVM()
    Color_Backing=@n2
    Color_Border=@n2.BorderColor_4_#2
    Color_Fill=@n2.FillColor_6_#3
    Color_Indicator=@n3
   )
   > SetIconColors(
    self=Get MyChildWidget()
    IN_BackingColor=@n2
    IN_BorderColor=@n2.BorderColor_4_#2
    IN_FillColor=@n2.FillColor_6_#3
   )
   > Set IdealTextColor(IdealTextColor=@n3)
  ]}
 UpdateTextByAffiliation {nodes [
   n1 = UpdateTextByAffiliation!() -> (then, TeamAffiliation:EFortTeamAffiliation Enum=Friendly) {is=title K2Node_FunctionEntry}
   > n2 = Get!(execute:Exec) -> (then:Exec, else:Exec, Value Bar Component:Creative Value Bar Component Object Reference)
   n3 = "Get (a copy)"(
    Array=Get Text Overrides!(self:Creative Value Bar Component Object Reference=@n2) -> (Text Overrides:Array of Strings)
    Dimension 1=Select!(Friendly:int=0, Neutral:int=1, Hostile:int=2, Index:EFortTeamAffiliation Enum=@n1) -> (ReturnValue:int)
   )
   > UpdateValue_Text_ObjectiveText!(execute, self:Device - Beacon View Model Object Reference=Get BadgeUIMVVM(), Value:Text="To Text (String)"(InString=@n3)) -> (then)
   > Set HasText(HasText:bool="To Boolean (Integer)"!(InInt:int=Len(S=@n3)) -> (ReturnValue:bool)) -> (Output_Get:bool)
   > Update Text Data() {is=title K2Node_CallFunction}
  ]}
 UpdateVisibilityByAffiliation {nodes [
   n1 = UpdateVisibilityByAffiliation!() -> (then, TeamAffiliation:EFortTeamAffiliation Enum=Friendly) {is=title K2Node_FunctionEntry}
   > n2 = Get()
   n3 = AND Boolean(
    A=IsShowTeamNeutral!(self:Creative Value Bar Component Object Reference=@n2) -> (ReturnValue:bool)
    B="Equal (Enum)"(A:EFortTeamAffiliation Enum=@n1, B:EFortTeamAffiliation Enum=Neutral)
   )
   n4 = AND Boolean(A=IsShowTeamFriendlies!(self:Creative Value Bar Component Object Reference=@n2) -> (ReturnValue:bool), B="Equal (Enum)"(A=@n1, B=Friendly))
   n5 = AND Boolean(A=IsShowTeamHostiles!(self:Creative Value Bar Component Object Reference=@n2) -> (ReturnValue:bool), B="Equal (Enum)"(A=@n1, B=Hostile))
   n6 = GetOwningPlayer()
   n7 = "Equal (Integer)"(
    A=GetShowTeamAsIndex!(self:Creative Value Bar Component Object Reference=@n2) -> (ReturnValue:int=0)
    B="To Integer (Byte)"(InByte=GetActorPvPTeamIndex(Actor=@n6) -> (TeamIndex=0)) -> (ReturnValue=0)
   ) {is=title K2Node_CallFunction}
   n8 = IsShowTeamValid!(self:Creative Value Bar Component Object Reference=@n2) -> (ReturnValue:bool)
   > n9 = Branch(Condition=OR Boolean(
     A=IsLocalPlayerRegistered!(self:Creative Value Bar Component Object Reference=@n2) -> (Valid:bool)
     B=IsShowTeamAny!(self:Creative Value Bar Component Object Reference=@n2) -> (ReturnValue:bool)
     C=AND Boolean(A=@n8, B=@n7)
     D=OR Boolean!(A=@n4, B=@n5, C=@n3) -> (ReturnValue)
    ))
   > UpdateVisibilityTimer!(execute, self, Visible:bool=true) -> (then) {is=title K2Node_CallFunction}
   UpdateVisibilityTimer(execute=@n9.else, Visible=false)
   > Disable HUD() {is=title K2Node_CallFunction}
  ]}
]
variables [
 BadgeUIMVVM {type=UEFN_BeaconBadge_ViewModel_C, default=null}
 bUseOwnerAffiliation {type=boolean, default=true}
 CurrentValuePercent {type=number}
 CurrentVisibility {type=boolean}
 CustomColor_BackingDifference {type=number, default=0.3}
 CustomColor_BorderDifference {type=number, default=0.2}
 CustomColor_BorderSwitchAmount {type=number, default=0.2}
 CustomColor_IndicatorSwitchAmount {type=number, default=0.35}
 DistanceTickTime {type=number, default=0.1}
 DistanceUpdateTimerHandle {type=TimerHandle}
 FontColorMap {default {
   Bold {colorUseRule=UseColor_Specified, specifiedColor=<rgba 1 1 1 0.9>}
   Subtle {colorUseRule=UseColor_Specified, specifiedColor=<rgba 1 1 1 0.75>}
  }}
 FontStyleMap {default {
   Bold {
    bForceMonospaced=false
    bMaterialIsStencil=false
    fontMaterial=null
    fontObject=&Burbank
    letterSpacing=1
    monospacedWidth=1
    size=20
    skewAmount=0
    typefaceFontName=Burbank-PlayerPower
    outlineSettings {
     bApplyOutlineToDropShadows=false
     bMiteredCorners=false
     bSeparateFillAlpha=false
     outlineMaterial=null
     outlineSize=1
     outlineColor=<rgba 0 0 0 1>
    }
   }
   Subtle {
    bForceMonospaced=false
    bMaterialIsStencil=false
    fontMaterial=null
    fontObject=&NotoSans
    letterSpacing=0
    monospacedWidth=1
    size=24
    skewAmount=0
    typefaceFontName=NotoSans-Bold
    outlineSettings {
     bApplyOutlineToDropShadows=false
     bMiteredCorners=false
     bSeparateFillAlpha=false
     outlineMaterial=null
     outlineSize=0
     outlineColor=<rgba 0 0 0 1>
    }
   }
  }}
 Fort Minigame Objective {type=FortMinigameObjectiveComponent, default=null}
 FortMinigameLogic {type=FortMinigameProgressComponent, default=null}
 HasText {type=boolean}
 HideTimer {type=TimerHandle}
 IconScaleWhenBackerHidden {type=number, default=1.3}
 IdealTextColor {type=LinearColor, default=<rgba 0 0 0 0>}
 isInLineOfSight {type=boolean}
 LastVisibility {type=boolean, default=true}
 My Actor {type=Actor, default=null}
 MyChildWidget {type=UserWidget, default=null}
 MyCurrentIndicatorTexture {type=Texture2D}
 MyIconMID {type=MaterialInstanceDynamic}
 OffscreenArrowTickTime {type=number, default=0.1}
 OffscreenArrowTimerHandle {type=TimerHandle}
 SetupCompleted {type=boolean}
 TeamColors {default {
   Friendly {
    backingColor=<rgba 0.004 0.0232 0.2159 1>
    borderColor=<rgba 0.0513 0.7379 1 1>
    contentColorAndOpacity=<rgba 0.5029 1 1 1>
    fillColor=<rgba 0 0.402 1 1>
   }
   Hostile {
    backingColor=<rgba 0.1845 0 0 1>
    borderColor=<rgba 1 0.4 0.1 1>
    contentColorAndOpacity=<rgba 1 0.7913 0.5029 1>
    fillColor=<rgba 0.8796 0.107 0.0212 1>
   }
   Neutral {
    backingColor=<rgba 0.0243 0.0243 0.0243 1>
    borderColor=<rgba 1 1 1 1>
    contentColorAndOpacity=<rgba 0.7913 1 0.7529 1>
    fillColor=<rgba 0.1746 0.7305 0.0782 1>
   }
  }}
 TeamIcons {default {
   Friendly=/Game/Athena/UI/InGame/Creative/PropertyEditor/Icon_Picker/Textures_V2/128x128/T_UI_IconLibrary_Gravestone_128
   Hostile=/Game/Athena/UI/InGame/Creative/PropertyEditor/Icon_Picker/Textures_V2/128x128/T_UI_IconLibrary_DogTags_128
   Neutral=/Game/Athena/UI/InGame/Creative/PropertyEditor/Icon_Picker/Textures/128x128/T_UI_IconLibrary_2_128
  }}
 Value Bar Component {type=Creative_ValueBar_Component_C, default=null}
 VisCheckTimerHandle {type=TimerHandle}
 Visibility Initial Tick Max {type=number, default=1}
 Visibility Tick Time {type=number, default=0.5}
]
widgets {
 info {inheritedWidgetCount=0, parentClass=&UserWidget, rootWidgetClass=&SizeBox, widgetCount=2}
 widgets [
  SizeBox_3 {widgetClassPath=&SizeBox, widget=SizeBox_3, properties {heightOverride=21.3612}}
  Overlay {widgetClassPath=&Overlay, bIsVariable=true, parent=SizeBox_3, widget=Overlay, slot {class=SizeBoxSlot}}
 ]
}
