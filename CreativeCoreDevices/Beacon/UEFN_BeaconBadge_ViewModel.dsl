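; View model for the UEFN Creative Beacon device's badge, holding the values that the badge displays.
; Custom update events each copy their inputs straight into variables for the objective label, distance text, icon dynamic material, and arrow rotation and visibility.
; A single color event sets the backing, border, fill and indicator colors together.
; The graph contains no logic beyond these setters, so values are computed by whatever calls the events.
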(event Custom|UpdateValue_Text_ObjectiveText
  (params (Value :type Text))
  (Variables|Default|SetText_Label Value))

(event Custom|UpdateValue_Indicator_Icon
  (params (Value :type Material%20Instance%20Dynamic%20Object%20Reference))
  (Variables|Default|SetIcon_Material Value))

(event Custom|UpdateValue_Text_DistanceText
  (params (Value :type Text))
  (Variables|Default|SetText_Distance Value))

(event Custom|UpdateValue_Colors
  (params (Color_Backing :type Linear%20Color%20Structure) (Color_Border :type Linear%20Color%20Structure) (Color_Fill :type Linear%20Color%20Structure) (Color_Indicator :type Linear%20Color%20Structure))
  (Variables|Default|SetColor_Backing Color_Backing)
  (Variables|Default|SetColor_Border Color_Border)
  (Variables|Default|SetColor_Fill Color_Fill)
  (Variables|Default|SetColor_Indicator Color_Indicator))

(event Custom|UpdateValue_ArrowRotationAngle
  (params (Value :type Float%20%28double-precision%29))
  (Variables|Default|SetArrow_RotationAngle Value))

(event Custom|UpdateValue_ArrowVisilbility
  (params (Value :type ESlateVisibility%20Enum :default "Visible"))
  (Variables|Default|SetArrow_Visiblity Value))
